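# Physics asset for the necklace on the medium male Cashier player character, setting up its secondary motion.
# Box bodies on the parent, 01 and 02 bones form a constrained chain from a small sphere on the base bone.
# Narrow swing and twist limits keep the chain fairly stiff, tightest at the last link.
# An unconstrained spine_05 body combines two capsules and a rotated box as an upper-torso collision shape.
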PhysicsAsset /BRCosmetics/Characters/Player/Male/Medium/Bodies/M_MED_Cashier/Meshes/M_MED_Cashier_Necklace_Physics
bodies {
 dyn_necklace_01 [{shapeType=Box, extentX=5.184, extentY=10.7764, extentZ=1.798, shapeName=dyn_necklace_01_box, center=<2.5297 0 0.0001>}]
 dyn_necklace_02 [{
   shapeType=Box
   extentX=3.0005
   extentY=6.6109
   extentZ=1.3718
   shapeName=dyn_necklace_02_box
   center=<1.0189 0 0>
   rotation=<rot -8.7484 0 0>
  }]
 dyn_necklace_base [{shapeType=Sphere, radius=0.5, shapeName=dyn_necklace_base_sphere}]
 dyn_necklace_parent [{
   shapeType=Box
   extentX=4.9519
   extentY=15.359
   extentZ=1.8703
   shapeName=dyn_necklace_parent_box
   center=<2.5918 0 0.249>
   rotation=<rot 1.4655 0 0>
  }]
 spine_05 [
  {shapeType=Capsule, length=28.4532, radius=10.3475, shapeName=spine_05_capsule, center=<-3.2211 0.2766 0>}
  {shapeType=Capsule, length=31.9298, radius=6.871, shapeName=spine_05_capsule, center=<6.3001 -2.4116 0.0001>}
  {shapeType=Box, extentX=22.8749, extentY=7.5, extentZ=32.1281, shapeName=spine_05_box, center=<-0.6602 5.6419 0>, rotation=<rot 0 -34.0383 0>}
 ]
}
constraints [
 {bone1Name=dyn_necklace_parent, bone2Name=dyn_necklace_base, swing1LimitDegrees=20, swing2LimitDegrees=20, twistLimitDegrees=3}
 {bone1Name=dyn_necklace_01, bone2Name=dyn_necklace_parent, swing1LimitDegrees=20, swing2LimitDegrees=20, twistLimitDegrees=3}
 {bone1Name=dyn_necklace_02, bone2Name=dyn_necklace_01, swing1LimitDegrees=15, swing2LimitDegrees=10, twistLimitDegrees=3}
]
