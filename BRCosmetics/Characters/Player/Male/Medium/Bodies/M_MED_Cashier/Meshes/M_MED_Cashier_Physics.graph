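# Physics asset for the M_MED_Cashier male player body cosmetic, providing collision and joints for its dangling drawstrings.
# Left and right four-capsule chains branch from dyn_string_parent, which joins a small dyn_string_base sphere.
# Constraints set narrow swing and twist limits along each chain, tightest in swing at the parent-to-base joint.
# Pelvis and thigh capsules have no constraints and cover only the lower body, apparently as colliders for the strings.
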PhysicsAsset /BRCosmetics/Characters/Player/Male/Medium/Bodies/M_MED_Cashier/Meshes/M_MED_Cashier_Physics
bodies {
 dyn_drawstring_01_l [{shapeType=Capsule, length=1.5906, radius=1.4968, shapeName=dyn_drawstring_01_l_capsule, center=<1.7049 0 0>, rotation=<rot 90 0 0>}]
 dyn_drawstring_01_r [{shapeType=Capsule, length=1.4381, radius=1.467, shapeName=dyn_drawstring_01_r_capsule, center=<1.7512 0 0>, rotation=<rot 90 0 0>}]
 dyn_drawstring_02_l [{shapeType=Capsule, length=1.8798, radius=1.6629, shapeName=dyn_drawstring_02_l_capsule, center=<1.7075 0 0>, rotation=<rot 90 0 0>}]
 dyn_drawstring_02_r [{shapeType=Capsule, length=2.6814, radius=1.5786, shapeName=dyn_drawstring_02_r_capsule, center=<1.935 0 -0.0001>, rotation=<rot 90 0 0>}]
 dyn_drawstring_03_l [{shapeType=Capsule, length=1.2598, radius=1.4349, shapeName=dyn_drawstring_03_l_capsule, center=<1.4609 0 0>, rotation=<rot 90 0 0>}]
 dyn_drawstring_03_r [{shapeType=Capsule, length=0.6025, radius=1.7825, shapeName=dyn_drawstring_03_r_capsule, center=<1.1417 0 0>, rotation=<rot 90 0 0>}]
 dyn_drawstring_04_l [{shapeType=Capsule, length=0.7304, radius=1.553, shapeName=dyn_drawstring_04_l_capsule, center=<1.0843 0 0>, rotation=<rot 90 0 0>}]
 dyn_drawstring_04_r [{shapeType=Capsule, length=1.4028, radius=1.8142, shapeName=dyn_drawstring_04_r_capsule, center=<1.7688 0 0>, rotation=<rot 90 0 0>}]
 dyn_string_base [{shapeType=Sphere, radius=0.5, shapeName=dyn_string_base_sphere}]
 dyn_string_parent [{shapeType=Capsule, length=1.3807, radius=1.4464, shapeName=dyn_string_parent_capsule, center=<1.571 0 0>, rotation=<rot 90 0 0>}]
 pelvis [
  {shapeType=Capsule, length=12.375, radius=12.375, shapeName=pelvis_capsule, center=<-6.0546 -3.4138 0>}
  {shapeType=Capsule, length=13.875, radius=10.875, shapeName=pelvis_capsule, center=<-3.7306 -2.0832 -0.0001>}
 ]
 thigh_l [{shapeType=Capsule, length=21, radius=9, shapeName=thigh_l_capsule, center=<-15.6977 0.8522 1.0401>, rotation=<rot 85.1482 -180 -180>}]
 thigh_r [{shapeType=Capsule, length=20, radius=10, shapeName=thigh_r_capsule, center=<15.7287 -0.976 -1.4171>, rotation=<rot 85.442 180 180>}]
}
constraints [
 {bone1Name=dyn_drawstring_01_r, bone2Name=dyn_string_parent, swing1LimitDegrees=20, swing2LimitDegrees=20, twistLimitDegrees=5}
 {bone1Name=dyn_drawstring_02_r, bone2Name=dyn_drawstring_01_r, swing1LimitDegrees=20, swing2LimitDegrees=20, twistLimitDegrees=5}
 {bone1Name=dyn_drawstring_03_r, bone2Name=dyn_drawstring_02_r, swing1LimitDegrees=20, swing2LimitDegrees=20, twistLimitDegrees=5}
 {bone1Name=dyn_drawstring_04_r, bone2Name=dyn_drawstring_03_r, swing1LimitDegrees=20, swing2LimitDegrees=20, twistLimitDegrees=5}
 {bone1Name=dyn_drawstring_01_l, bone2Name=dyn_string_parent, swing1LimitDegrees=20, swing2LimitDegrees=20, twistLimitDegrees=5}
 {bone1Name=dyn_drawstring_02_l, bone2Name=dyn_drawstring_01_l, swing1LimitDegrees=20, swing2LimitDegrees=20, twistLimitDegrees=5}
 {bone1Name=dyn_drawstring_03_l, bone2Name=dyn_drawstring_02_l, swing1LimitDegrees=20, swing2LimitDegrees=20, twistLimitDegrees=5}
 {bone1Name=dyn_drawstring_04_l, bone2Name=dyn_drawstring_03_l, swing1LimitDegrees=20, swing2LimitDegrees=20, twistLimitDegrees=5}
 {bone1Name=dyn_string_parent, bone2Name=dyn_string_base, swing1LimitDegrees=10, swing2LimitDegrees=10, twistLimitDegrees=5}
]
