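# Post-process style AnimBP for the Cashier player body, adding corrective deformation and secondary motion to the pose copied from the attached parent mesh.
# Two Control Rigs feed shoulder bone scaling driven by clavicle-to-upperarm twist.
# RigidBody simulates necklace and body physics; Look At and a crouch-driven rotation adjust drawstrings.
# Weighted constraints anchor Anim Dynamics shoelaces and zippers, with jump and fall events setting a fixed force on the laces.
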AnimBlueprint /BRCosmetics/Characters/Player/Male/Medium/Bodies/M_MED_Cashier/Meshes/M_MED_Cashier_AnimBP
target_skeleton=&/M_MED_Cashier_Skeleton
graphs [
 AnimGraph {nodes [
   n1 = Control Rig(Source=Copy Pose From Mesh() {bUseAttachedParent=true}) {
    alphaBoolBlend {blendOption=Linear}
    controlRigAssetReference {blueprintRigClass=&/M_MED_Cashier_Deform_CtrlRig_twists_2.M_MED_Cashier_Deform_CtrlRig_twists_2_C}
   }
   n2 = Control Rig!(
    Source=@n1
    Alpha
    KneeRot_fr_l_114deg:Vector
    KneeTrans_fr_l_114deg:Vector="0.258,-3.591,0.009"
    KneeScale_fr_l_114deg:Vector="1,1,1"
    KneeScale_fr_l_130deg:Vector="1,1,1"
    KneeRot_fr_l_130deg:Vector="0,0,-51.2"
    KneeTrans_fr_l_130deg:Vector="0.258,-3.591,0.009"
   ) -> (Pose) {
    alphaBoolBlend {blendOption=Linear}
    controlRigAssetReference {blueprintRigClass=&/M_MED_Cashier_Deform_CtrlRig_sculpters.M_MED_Cashier_Deform_CtrlRig_sculpters_C}
   }
   n3 = Get Shoulder_R_Alpha() -> (Shoulder_R_Alpha:double)
   n4 = "Transform (Modify) Bone - Bone: deform_deltoid_fr_r"(ComponentPose=Local To Component(LocalPose=@n2), Alpha=@n3, Translation="0,0,0", Rotation="0,0,0", Scale="2,2,1.15") {
    scaleMode=BMM_Replace
    scaleSpace=BCS_BoneSpace
    alphaBoolBlend {blendOption=Linear}
    alphaScaleBiasClamp {bInterpResult=true, interpSpeedDecreasing=20, interpSpeedIncreasing=20}
    boneToModify {boneName=deform_deltoid_fr_r}
    scale=<2 2 1.15>
   }
   n5 = "Transform (Modify) Bone - Bone: upperarm_twist_02_r"(ComponentPose=@n4, Alpha=@n3, Translation="0,0,0", Rotation="0,0,0", Scale="1.15,1.15,1.15") {
    scaleMode=BMM_Replace
    scaleSpace=BCS_BoneSpace
    alphaBoolBlend {blendOption=Linear}
    alphaScaleBiasClamp {bInterpResult=true, interpSpeedDecreasing=20, interpSpeedIncreasing=20}
    boneToModify {boneName=upperarm_twist_02_r}
    scale=<1.15 1.15 1.15>
   }
   n6 = RigidBody(ComponentPose=@n5) {
    bClampLinearTranslationLimitToRefPose=true
    bOverrideWorldGravity=true
    overridePhysicsAsset=&/M_MED_Cashier_Necklace_Physics
    simulationSpace=BaseBoneSpace
    alphaBoolBlend {blendOption=Linear}
    baseBoneRef {boneName=spine_03}
    componentAppliedLinearAccClamp=<6000 6000 6000>
    componentLinearAccScale=<0.25 0.25 0.25>
    componentLinearVelScale=<0.25 0.25 0.25>
    overrideWorldGravity {z=-980}
   }
   n7 = RigidBody(ComponentPose=@n6) {
    bClampLinearTranslationLimitToRefPose=true
    bOverrideWorldGravity=true
    overridePhysicsAsset=&/M_MED_Cashier_Physics
    simulationSpace=BaseBoneSpace
    alphaBoolBlend {blendOption=Linear}
    alphaScaleBiasClamp {bClampResult=true}
    baseBoneRef {boneName=root}
    componentAppliedLinearAccClamp=<6000 6000 6000>
    componentLinearAccScale=<0.1 0.1 0.1>
    componentLinearVelScale=<0.1 0.1 0.1>
    overrideWorldGravity {z=-980}
   }
   n8 = Look At - Bone: dyn_drawstring_01_l(ComponentPose=@n7, Alpha=0.3) {
    lookAtClamp=60
    alphaBoolBlend {blendOption=Linear}
    boneToModify {boneName=dyn_drawstring_01_l}
    lookAtTarget {boneReference {boneName=root}}
    lookAt_Axis {axis {y=0, z=1}}
    lookUp_Axis {axis {x=0}}
   }
   n9 = Look At - Bone: dyn_drawstring_01_r(ComponentPose=@n8, Alpha=0.3) {
    lookAtClamp=60
    alphaBoolBlend {blendOption=Linear}
    boneToModify {boneName=dyn_drawstring_01_r}
    lookAtTarget {boneReference {boneName=root}}
    lookAt_Axis {axis {y=0, z=1}}
    lookUp_Axis {axis {x=0}}
   }
   n10 = "Transform (Modify) Bone - Bone: dyn_string_parent"(ComponentPose=@n9, Alpha=Get Crouch_Alpha() -> (Crouch_Alpha:double), Translation="0,0,0", Rotation="-8,-25,0", Scale="1,1,1") {
    rotationMode=BMM_Additive
    rotationSpace=BCS_BoneSpace
    alphaBoolBlend {blendOption=Linear}
    alphaScaleBiasClamp {bInterpResult=true, interpSpeedDecreasing=3, interpSpeedIncreasing=5}
    boneToModify {boneName=dyn_string_parent}
    rotation {pitch=-8, yaw=-25}
   }
   n11 = "Constraint - dyn_shoelace_parent_r (PPPP)"!(
    ComponentPose=@n10
    Alpha=1
    ConstraintWeights_0=0.0004
    ConstraintWeights_1:float=0.0066
    ConstraintWeights_2:float=0.0705
    ConstraintWeights_3:float=0.9224
   ) -> (Pose) {
    alphaBoolBlend {blendOption=Linear}
    boneToModify {boneName=dyn_shoelace_parent_r}
    constraintSetup [
     {targetBone {boneName=ball_r}}
     {targetBone {boneName=calf_twist_02_r}}
     {targetBone {boneName=calf_twist_01_r}}
     {targetBone {boneName=foot_r}}
    ]
    constraintWeights [1,1,1,1]
   }
   n12 = Get Velocity() -> (Velocity:Vector)
   n13 = Anim Dynamics - Bone: dyn_shoelace_inner_01_r(ComponentPose=@n11, Alpha=1, ExternalForce=@n12) {
    bGravityOverrideInSimSpace=true
    bOverrideAngularDamping=true
    bOverrideLinearDamping=true
    simulationSpace=RootRelative
    alphaBoolBlend {blendOption=Linear}
    alphaScaleBiasClamp {bInterpResult=true}
    boundBone {boneName=dyn_shoelace_inner_01_r}
    componentAppliedLinearAccClamp=<3000 3000 3000>
    componentLinearAccScale=<1.5 1.5 1.5>
    componentLinearVelScale=<1.5 1.5 1.5>
    gravityOverride=<1 1 1>
    physicsBodyDefinitions [{
      sphereCollisionRadius=0
      boundBone {boneName=dyn_shoelace_inner_01_r}
      boxExtents=<7 4 3>
      constraintSetup {angularLimitsMax=<0 40 40>, angularLimitsMin=<0 -5 -30>, angularTarget=<0 0 0>}
      localJointOffset=<1 0 0>
     }]
    relativeSpaceBone {boneName=root}
   }
   n14 = Anim Dynamics - Bone: dyn_shoelace_outter_01_r(ComponentPose=@n13, Alpha=1, ExternalForce=@n12) {
    angularDampingOverride=0.9999
    bGravityOverrideInSimSpace=true
    bOverrideAngularDamping=true
    bOverrideLinearDamping=true
    linearDampingOverride=0.99
    simulationSpace=RootRelative
    alphaBoolBlend {blendOption=Linear}
    boundBone {boneName=dyn_shoelace_outter_01_r}
    componentAppliedLinearAccClamp=<6000 6000 6000>
    componentLinearAccScale=<0.5 0.5 0.5>
    componentLinearVelScale=<0.5 0.5 0.5>
    gravityOverride=<1 1 1>
    physicsBodyDefinitions [{
      sphereCollisionRadius=0
      boundBone {boneName=dyn_shoelace_outter_01_r}
      boxExtents=<7 3 2>
      constraintSetup {angularLimitsMax=<0 30 20>, angularLimitsMin=<0 -2 -20>, angularTarget=<0 0 0>}
      localJointOffset=<1.25 0 0>
     }]
    relativeSpaceBone {boneName=pelvis}
   }
   n15 = "Constraint - dyn_shoelace_parent_l (PPP)"!(ComponentPose=@n14, Alpha=1, ConstraintWeights_0=0.071, ConstraintWeights_1:float=0.9285, ConstraintWeights_2:float=0.0004) -> (Pose) {
    alphaBoolBlend {blendOption=Linear}
    boneToModify {boneName=dyn_shoelace_parent_l}
    constraintSetup [{targetBone {boneName=calf_twist_01_l}}, {targetBone {boneName=foot_l}}, {targetBone {boneName=ball_l}}]
    constraintWeights [1,1,1]
   }
   n16 = Get Velocity()
   n17 = Anim Dynamics - Bone: dyn_shoelace_inner_01_l(ComponentPose=@n15, Alpha=1, ExternalForce=@n16) {
    bGravityOverrideInSimSpace=true
    bOverrideAngularDamping=true
    bOverrideLinearDamping=true
    simulationSpace=RootRelative
    alphaBoolBlend {blendOption=Linear}
    alphaScaleBiasClamp {bInterpResult=true}
    boundBone {boneName=dyn_shoelace_inner_01_l}
    componentAppliedLinearAccClamp=<3000 3000 3000>
    componentLinearAccScale=<1.5 1.5 1.5>
    componentLinearVelScale=<1.5 1.5 1.5>
    gravityOverride=<1 1 1>
    physicsBodyDefinitions [{
      sphereCollisionRadius=0
      boundBone {boneName=dyn_shoelace_inner_01_l}
      boxExtents=<7 3 2>
      constraintSetup {angularLimitsMax=<40 40 30>, angularLimitsMin=<-20 -20 -30>, angularTarget=<0 0 0>}
      localJointOffset=<1.25 0 0>
     }]
    relativeSpaceBone {boneName=root}
   }
   n18 = Anim Dynamics - Bone: dyn_shoelace_outter_01_l(ComponentPose=@n17, Alpha=1, ExternalForce=@n16) {
    bGravityOverrideInSimSpace=true
    bOverrideAngularDamping=true
    bOverrideLinearDamping=true
    simulationSpace=RootRelative
    alphaBoolBlend {blendOption=Linear}
    boundBone {boneName=dyn_shoelace_outter_01_l}
    componentAppliedLinearAccClamp=<3000 3000 3000>
    componentLinearAccScale=<1.5 1.5 1.5>
    componentLinearVelScale=<1.5 1.5 1.5>
    gravityOverride=<1 1 1>
    physicsBodyDefinitions [{
      sphereCollisionRadius=0
      boundBone {boneName=dyn_shoelace_outter_01_l}
      boxExtents=<7 3 2>
      constraintSetup {angularLimitsMax=<40 40 30>, angularLimitsMin=<-20 -3 -30>, angularTarget=<0 0 0>}
      localJointOffset=<1.25 0 0>
     }]
    relativeSpaceBone {boneName=pelvis}
   }
   n19 = "Constraint - dyn_zipper_base_l (PPPPP)"!(
    ComponentPose=@n18
    Alpha=1
    ConstraintWeights_0=0.2935
    ConstraintWeights_1:float=0.0141
    ConstraintWeights_2:float=0.4893
    ConstraintWeights_3:float=0.0144
    ConstraintWeights_4:float=0.1888
   ) -> (Pose) {
    alphaBoolBlend {blendOption=Linear}
    boneToModify {boneName=dyn_zipper_base_l}
    constraintSetup [
     {targetBone {boneName=deform_thigh_fr_l}}
     {targetBone {boneName=spine_01}}
     {targetBone {boneName=thigh_twist_01_l}}
     {targetBone {boneName=thigh_twist_01_r}}
     {targetBone {boneName=pelvis}}
    ]
    constraintWeights [1,1,1,1,1]
   }
   n20 = Anim Dynamics - Bone: dyn_zipper_parent_l(ComponentPose=@n19, Alpha=1, ExternalForce="0,0,0") {
    angularDampingOverride=0.9999
    bGravityOverrideInSimSpace=true
    bOverrideAngularDamping=true
    bOverrideLinearDamping=true
    linearDampingOverride=0.9999
    alphaBoolBlend {blendOption=Linear}
    boundBone {boneName=dyn_zipper_parent_l}
    componentLinearAccScale=<0.25 0.25 0.25>
    componentLinearVelScale=<0.25 0.25 0.25>
    physicsBodyDefinitions [{
      sphereCollisionRadius=0
      boundBone {boneName=dyn_zipper_parent_l}
      boxExtents=<7 3 2>
      constraintSetup {angularLimitsMax=<0 30 20>, angularLimitsMin=<0 -5 -20>, angularTarget=<0 0 0>}
      localJointOffset=<1.5 0 0>
     }]
    relativeSpaceBone {boneName=pelvis}
   }
   n21 = "Constraint - dyn_zipper_base_r (PPPPP)"!(
    ComponentPose=@n20
    Alpha=1
    ConstraintWeights_0=0.0116
    ConstraintWeights_1:float=0.016
    ConstraintWeights_2:float=0.2934
    ConstraintWeights_3:float=0.4917
    ConstraintWeights_4:float=0.1874
   ) -> (Pose) {
    alphaBoolBlend {blendOption=Linear}
    boneToModify {boneName=dyn_zipper_base_r}
    constraintSetup [
     {targetBone {boneName=spine_01}}
     {targetBone {boneName=thigh_twist_01_l}}
     {targetBone {boneName=deform_thigh_fr_r}}
     {targetBone {boneName=thigh_twist_01_r}}
     {targetBone {boneName=pelvis}}
    ]
    constraintWeights [1,1,1,1,1]
   }
   n22 = Anim Dynamics - Bone: dyn_zipper_parent_r(ComponentPose=@n21, Alpha=1, ExternalForce="0,0,0") {
    angularDampingOverride=0.9999
    bGravityOverrideInSimSpace=true
    bOverrideAngularDamping=true
    bOverrideLinearDamping=true
    linearDampingOverride=0.9999
    alphaBoolBlend {blendOption=Linear}
    boundBone {boneName=dyn_zipper_parent_r}
    componentLinearAccScale=<0.25 0.25 0.25>
    componentLinearVelScale=<0.25 0.25 0.25>
    physicsBodyDefinitions [{
      sphereCollisionRadius=0
      boundBone {boneName=dyn_zipper_parent_r}
      boxExtents=<7 3 2>
      constraintSetup {angularLimitsMax=<0 30 20>, angularLimitsMin=<0 -5 -20>, angularTarget=<0 0 0>}
      localJointOffset=<1.5 0 0>
     }]
    relativeSpaceBone {boneName=pelvis}
   }
   Output Pose(Result=Component To Local(ComponentPose=@n22)) {layerGroup=None}
  ]}
 EventGraph {nodes [
   Event BlueprintInitializeAnimation()
   > Set Skel_Mesh(Skel_Mesh=GetPartSkeletalMeshComponent())
   Event OnBeginJumping()
   > Set Velocity(Velocity:Vector="1500,1500,1500") -> (Output_Get:Vector)
   Event OnEndJumping()
   > Set Velocity(Velocity="0, 0, 0")
   Event OnEndFalling()
   > Set Velocity(Velocity="0, 0, 0")
   Event OnBeginFalling()
   > Set Velocity(Velocity="1500,1500,1500")
   Event BlueprintUpdateAnimation()
   > n1 = Sequence()
   > Set Shoulder_R_Alpha(Shoulder_R_Alpha:double=GetAlphaFromReferenceTransformDelta(
     SkeletalMeshComponent=Get Skel_Mesh()
     BoneName=clavicle_r
     BaseName=upperarm_r
     DeltaComponentType=QuaternionTwist
     InputRange=(X=-5,Y=120)
     TwistAxis="0,0,1"
    )) -> (Output_Get:double)
   Set Crouch_Alpha(execute=@n1.then_1, Crouch_Alpha:double=SelectFloat(A=1, B=0, bPickA=Get bIsCrouching())) -> (Output_Get:double)
  ]}
]
variables [
 Crouch_Alpha {type=number}
 Shoulder_R_Alpha {type=number}
 Skel_Mesh {type=SkeletalMeshComponent}
 Velocity {type=Vector}
]
