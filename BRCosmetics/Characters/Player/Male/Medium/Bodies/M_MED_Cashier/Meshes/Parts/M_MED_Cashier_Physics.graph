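# Physics asset for the medium male Cashier cosmetic body part, set up for two dynamic hair bone chains and the head.
# Each chain has a small sphere on its base bone and a box on its parent bone, joined by a constraint.
# Both constraints lock twist and swing 2, leaving swing 1 as the only unlocked angular axis.
# A capsule on the head bone supplies a head collision shape, with no constraint attached.
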PhysicsAsset /BRCosmetics/Characters/Player/Male/Medium/Bodies/M_MED_Cashier/Meshes/Parts/M_MED_Cashier_Physics
bodies {
 dyn_hair_01_base [{shapeType=Sphere, radius=0.5, shapeName=dyn_hair_01_base_sphere}]
 dyn_hair_01_parent [{shapeType=Box, extentX=2.4388, extentY=2.2153, extentZ=0.7531, shapeName=dyn_hair_01_parent_box, center=<1.0685 0 0>}]
 dyn_hair_02_base [{shapeType=Sphere, radius=0.5, shapeName=dyn_hair_02_base_sphere}]
 dyn_hair_02_parent [{shapeType=Box, extentX=2.5912, extentY=2.4792, extentZ=1.077, shapeName=dyn_hair_02_parent_box, center=<1.1231 0 0>}]
 head [{shapeType=Capsule, length=4.1889, radius=5.0189, shapeName=head_capsule, center=<5.3623 5.0174 -0.4236>, rotation=<rot 90 0 12.788>}]
}
constraints [
 {
  bone1Name=dyn_hair_02_parent
  bone2Name=dyn_hair_02_base
  swing1LimitDegrees=5
  swing2LimitDegrees=3
  swing2Motion=Locked
  twistLimitDegrees=2
  twistMotion=Locked
 }
 {
  bone1Name=dyn_hair_01_parent
  bone2Name=dyn_hair_01_base
  swing1LimitDegrees=5
  swing2LimitDegrees=3
  swing2Motion=Locked
  twistLimitDegrees=2
  twistMotion=Locked
 }
]
